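# QuadLowPass is a reusable MetaSound patch that low-pass filters a four-channel quad signal (FL, FR, SL, SR).
# Each channel runs through its own One-Pole Low Pass Filter to the matching output.
# A single Cutoff Frequency input drives all four filters, defaulting to 20 kHz, so filtering is minimal until the value is lowered.
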MetaSoundPatch /FortMetasound/Patches/QuadLowPass
inputs [
 Cutoff Frequency {type=Float}
 In FL {type=Audio}
 In FR {type=Audio}
 In SL {type=Audio}
 In SR {type=Audio}
]
outputs [
 Out FL {type=Audio}
 Out FR {type=Audio}
 Out SL {type=Audio}
 Out SR {type=Audio}
]
pages [Default {
  input_defaults {Cutoff Frequency=20000, In FL=null, In FR=null, In SL=null, In SR=null}
  nodes [
   OnePoleLowPassFilter_0 = UE.One-Pole Low Pass Filter:Audio(Cutoff Frequency=^Cutoff Frequency, In=^In FR)
   OnePoleLowPassFilter_1 = UE.One-Pole Low Pass Filter:Audio(Cutoff Frequency=^Cutoff Frequency, In=^In FL)
   OnePoleLowPassFilter_2 = UE.One-Pole Low Pass Filter:Audio(Cutoff Frequency=^Cutoff Frequency, In=^In SR)
   OnePoleLowPassFilter_3 = UE.One-Pole Low Pass Filter:Audio(Cutoff Frequency=^Cutoff Frequency, In=^In SL)
  ]
  links [
   {dst="@Out FR", src=OnePoleLowPassFilter_0.Out}
   {dst="@Out FL", src=OnePoleLowPassFilter_1.Out}
   {dst="@Out SR", src=OnePoleLowPassFilter_2.Out}
   {dst="@Out SL", src=OnePoleLowPassFilter_3.Out}
  ]
 }]
