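# Looping quad MetaSound rain ambience that blends random exterior and interior beds using the worldization enclosure modulator, with Amb.Height scaling the exterior.
# Upward enclosure drives a low-pass on the interior bed.
# The mix is azimuth-panned by listener orientation with slow LFO gain drift; a height-scaled splash loop joins the front channels post-pan before the fade-in envelope.
# The Mobile page uses fixed stereo beds, omitting filter and splash.
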MetaSoundSource /AmbientAudioContent/MetaSounds/Loops/MSS_Ambient_Rain_Quad
inputs [
 Amb.Height {type=Float}
 FadeInTime {type=Time}
 Sounds.Exterior {type=WaveAsset:Array}
 Sounds.Interior {type=WaveAsset:Array}
 Sounds.Splash {type=WaveAsset:Array}
 UE.Listener.Orientation.Azimuth {type=Float}
 UE.Listener.Orientation.Elevation {type=Float}
 UE.Source.OnPlay {type=Trigger}
 Volume.Exterior {type=Float}
 Volume.Interior {type=Float}
 Volume.QuadBed {type=Float}
 Volume.Splash {type=Float}
]
interfaces [UE.Listener.Orientation:1.0, UE.OutputFormat.Quad:1.0, UE.Source:1.1]
outputs [
 "UE.OutputFormat.Quad.Audio:0" {type=Audio}
 "UE.OutputFormat.Quad.Audio:1" {type=Audio}
 "UE.OutputFormat.Quad.Audio:4" {type=Audio}
 "UE.OutputFormat.Quad.Audio:5" {type=Audio}
]
pages [
 Default {
  is_build_graph=true
  input_defaults {
   Amb.Height=1
   FadeInTime=2
   UE.Listener.Orientation.Azimuth=0
   UE.Listener.Orientation.Elevation=0
   UE.Source.OnPlay=null
   Volume.Exterior=0.827
   Volume.Interior=0.5
   Volume.QuadBed=1
   Volume.Splash=0.2
   Sounds.Exterior [/AmbientAudioContent/Waves/Biomes/Ambience_Neon_City_Ext_Rain_Quad_Default_Loop]
   Sounds.Interior [
    /AmbientAudioContent/Waves/POI/Ambience_Neon_City_Int_Rain_Interior_Light_01_Quad
    /AmbientAudioContent/Waves/POI/Ambience_Neon_City_Int_Rain_Interior_Light_02_Quad
    /AmbientAudioContent/Waves/POI/Ambience_Neon_City_Int_Rain_Interior_Light_03_Quad
   ]
   Sounds.Splash [
    /AmbientAudioContent/Waves/SpottedLoops/Spot_Water_Splashing_Concrete_Int_Loop_02
    /AmbientAudioContent/Waves/SpottedLoops/Spot_Water_Splashing_Concrete_Int_Loop_01
   ]
  }
  nodes [
   ADSREnvelopeADSREnvelope_0 = ADSR Envelope.ADSR Envelope:Float(Attack Time=^FadeInTime, Trigger Attack=^UE.Source.OnPlay) {also {defaults {Sustain Level=1}}}
   n1 = InitVariable.Float()
   n2 = MapRange.MapRange:Float(In=Modulation.GetModulatorValue() {also {defaults {Modulator=/AudioWorldization/Modulation/AudioWorldization_Quadrant_Up_Bus_Enclosure}}}.Out) {also {defaults {Out Range A=20000, Out Range B=12000}}}
   ArrayRandomGet_1 = Array.Random Get:WaveAsset:Array(In Array=^Sounds.Interior, Next=^UE.Source.OnPlay)
   WavePlayer_1 = UE.Wave Player:Quad(Play=@ArrayRandomGet_1.On Next, Wave Asset=@ArrayRandomGet_1.Value) {also {defaults {Loop="<MetaSoundBuilderResult.SUCCEEDED: 0>"}}}
   n3 = /FortMetasound/Patches/QuadLowPass(
    Cutoff Frequency=VariableAccessor.Float(Variable=VariableMutator.Float(Variable=@n1.Variable, Value=@n2.Out Value).Variable).Value
    In FL=@WavePlayer_1.Out Front Left
    In FR=@WavePlayer_1.Out Front Right
    In SL=@WavePlayer_1.Out Side Left
    In SR=@WavePlayer_1.Out Side Right
   )
   n4 = Modulation.GetModulatorValue() {also {defaults {Modulator=&AudioWorldization_Bus_Enclosure}}}
   n5 = MapRange.MapRange:Float(In=@n4.Out) {also {defaults {In Range A=0.4}}}
   n6 = VariableAccessor.Float(Variable=VariableMutator.Float(Variable=InitVariable.Float().Variable, Value=UE.Multiply:Float(PrimaryOperand=^Volume.Interior, AdditionalOperands=@n5.Out Value).Out).Variable)
   n7 = Clamp.Clamp:Float(In=^Amb.Height) {also {defaults {Max=1, Min=0.5}}}
   n8 = UE.Multiply:Float(
    PrimaryOperand=^Volume.Exterior
    AdditionalOperands=MapRange.MapRange:Float(In=@n4.Out) {also {defaults {Out Range A=1, Out Range B=0.2}}}.Out Value
   )
   n9 = VariableAccessor.Float(Variable=VariableMutator.Float(Variable=InitVariable.Float().Variable, Value=UE.Multiply:Float(AdditionalOperands=@n7.Value, PrimaryOperand=@n8.Out).Out).Variable)
   ArrayRandomGet_0 = Array.Random Get:WaveAsset:Array(In Array=^Sounds.Exterior, Next=^UE.Source.OnPlay) {also {defaults {Weights []}}}
   WavePlayer_0 = UE.Wave Player:Quad(Play=@ArrayRandomGet_0.On Next, Wave Asset=@ArrayRandomGet_0.Value) {also {defaults {Loop="<MetaSoundBuilderResult.SUCCEEDED: 0>"}}}
   AudioMixerAudioMixerStereo2_1 = "AudioMixer.Audio Mixer (Stereo, 2)"(
    In 0 L=@n3.Out FL
    In 0 R=@n3.Out FR
    Gain 0=@n6.Value
    Gain 1=@n9.Value
    In 1 L=@WavePlayer_0.Out Front Left
    In 1 R=@WavePlayer_0.Out Front Right
   )
   n10 = VariableAccessor.Float(Variable=@n9.Variable)
   AudioMixerAudioMixerStereo2_2 = "AudioMixer.Audio Mixer (Stereo, 2)"(
    In 0 L=@n3.Out SL
    In 0 R=@n3.Out SR
    Gain 0=VariableAccessor.Float(Variable=@n6.Variable).Value
    Gain 1=@n10.Value
    In 1 L=@WavePlayer_0.Out Side Left
    In 1 R=@WavePlayer_0.Out Side Right
   )
   n11 = VariableMutator.Float(Variable=InitVariable.Float().Variable, Value=UE.LFO:Audio() {also {defaults {Frequency=0.05, Min Value=0.9}}}.Out)
   n12 = &MSP_AzimuthPannerQuad(
    Azimuth=^UE.Listener.Orientation.Azimuth
    In Front Left=@AudioMixerAudioMixerStereo2_1.Out L
    In Front Right=@AudioMixerAudioMixerStereo2_1.Out R
    In Side Left=@AudioMixerAudioMixerStereo2_2.Out L
    In Side Right=@AudioMixerAudioMixerStereo2_2.Out R
    Gain=UE.Multiply:Float(AdditionalOperands=^Volume.QuadBed, PrimaryOperand=VariableAccessor.Float(Variable=@n11.Variable).Value).Out
   )
   Multiply_0 = UE.Multiply:Audio by Float(AdditionalOperands=@ADSREnvelopeADSREnvelope_0.Out Envelope, PrimaryOperand=@n12.Out Side Right)
   n13 = VariableMutator.Float(Variable=InitVariable.Float().Variable, Value=UE.Multiply:Float(AdditionalOperands=^Amb.Height, PrimaryOperand=^Volume.Splash).Out)
   ArrayRandomGet_2 = Array.Random Get:WaveAsset:Array(In Array=^Sounds.Splash, Next=^UE.Source.OnPlay) {also {defaults {Weights []}}}
   WavePlayer_2 = UE.Wave Player:Stereo(Play=@ArrayRandomGet_2.On Next, Wave Asset=@ArrayRandomGet_2.Value) {also {defaults {Loop="<MetaSoundBuilderResult.SUCCEEDED: 0>"}}}
   AudioMixerAudioMixerStereo2_0 = "AudioMixer.Audio Mixer (Stereo, 2)"(
    In 0 L=@n12.Out Front Left
    In 0 R=@n12.Out Front Right
    Gain 1=UE.Multiply:Float(AdditionalOperands=VariableAccessor.Float(Variable=@n13.Variable).Value, PrimaryOperand=VariableAccessor.Float(Variable=@n10.Variable).Value).Out
    In 1 L=@WavePlayer_2.Out Left
    In 1 R=@WavePlayer_2.Out Right
   )
   Multiply_1 = UE.Multiply:Audio by Float(AdditionalOperands=@ADSREnvelopeADSREnvelope_0.Out Envelope, PrimaryOperand=@AudioMixerAudioMixerStereo2_0.Out R)
   Multiply_2 = UE.Multiply:Audio by Float(AdditionalOperands=@ADSREnvelopeADSREnvelope_0.Out Envelope, PrimaryOperand=@n12.Out Side Left)
   Multiply_3 = UE.Multiply:Audio by Float(AdditionalOperands=@ADSREnvelopeADSREnvelope_0.Out Envelope, PrimaryOperand=@AudioMixerAudioMixerStereo2_0.Out L)
  ]
  links [
   {dst="@UE.OutputFormat.Quad.Audio:5", src=Multiply_0.Out}
   {dst="@UE.OutputFormat.Quad.Audio:1", src=Multiply_1.Out}
   {dst="@UE.OutputFormat.Quad.Audio:4", src=Multiply_2.Out}
   {dst="@UE.OutputFormat.Quad.Audio:0", src=Multiply_3.Out}
  ]
 }
 Mobile {
  input_defaults {
   Sounds.Exterior [/AmbientAudioContent/Waves/Biomes/Ambience_Neon_City_Ext_Rain_Stereo_Default_Loop]
   Sounds.Interior [/AmbientAudioContent/Waves/POI/Ambience_Neon_City_Int_Rain_Interior_Light_02_Stereo]
   Sounds.Splash [&Spot_Water_Splashing_Concrete_Int_Loop_02]
  }
  nodes [
   InitVariable.Float()
   InitVariable.Float()
   InitVariable.Float()
   ADSREnvelopeADSREnvelope_0 = ADSR Envelope.ADSR Envelope:Float(Attack Time=^FadeInTime, Trigger Attack=^UE.Source.OnPlay) {also {defaults {Sustain Level=1}}}
   n1 = InitVariable.Float()
   n2 = Modulation.GetModulatorValue() {also {defaults {Modulator=&AudioWorldization_Bus_Enclosure}}}
   n3 = UE.Multiply:Float(PrimaryOperand=^Volume.Interior, AdditionalOperands=MapRange.MapRange:Float(In=@n2.Out) {also {defaults {In Range A=0.4}}}.Out Value)
   n4 = InitVariable.Float()
   n5 = UE.Multiply:Float(
    AdditionalOperands=Clamp.Clamp:Float(In=^Amb.Height) {also {defaults {Max=1, Min=0.5}}}.Value
    PrimaryOperand=UE.Multiply:Float(
     PrimaryOperand=^Volume.Exterior
     AdditionalOperands=MapRange.MapRange:Float(In=@n2.Out) {also {defaults {Out Range A=1, Out Range B=0.2}}}.Out Value
    ).Out
   )
   WavePlayer_0 = UE.Wave Player:Stereo(Play=^UE.Source.OnPlay, Wave Asset=Array.Get:WaveAsset:Array(Array=^Sounds.Interior, Trigger=^UE.Source.OnPlay).Element) {also {defaults {Loop="<MetaSoundBuilderResult.SUCCEEDED: 0>"}}}
   WavePlayer_1 = UE.Wave Player:Stereo(
    Play=^UE.Source.OnPlay
    Wave Asset=Array.Get:WaveAsset:Array(Array=^Sounds.Exterior, Trigger=^UE.Source.OnPlay) {also {defaults {Index=0}}}.Element
   ) {also {defaults {Loop="<MetaSoundBuilderResult.SUCCEEDED: 0>"}}}
   AudioMixerAudioMixerStereo2_0 = "AudioMixer.Audio Mixer (Stereo, 2)"(
    Gain 0=VariableAccessor.Float(Variable=VariableMutator.Float(Variable=@n1.Variable, Value=@n3.Out).Variable).Value
    Gain 1=VariableAccessor.Float(Variable=VariableMutator.Float(Variable=@n4.Variable, Value=@n5.Out).Variable).Value
    In 0 L=@WavePlayer_0.Out Left
    In 0 R=@WavePlayer_0.Out Right
    In 1 L=@WavePlayer_1.Out Left
    In 1 R=@WavePlayer_1.Out Right
   )
   n6 = &AzimuthPannerStereo(Azimuth=^UE.Listener.Orientation.Azimuth, In L=@AudioMixerAudioMixerStereo2_0.Out L, In R=@AudioMixerAudioMixerStereo2_0.Out R)
   n7 = InitVariable.Float()
   n8 = UE.LFO:Audio() {also {defaults {Frequency=0.05, Min Value=0.9}}}
   n9 = UE.Multiply:Float(
    AdditionalOperands=^Volume.QuadBed
    PrimaryOperand=VariableAccessor.Float(Variable=VariableMutator.Float(Variable=@n7.Variable, Value=@n8.Out).Variable).Value
   )
   Multiply_2 = UE.Multiply:Audio by Float(PrimaryOperand=@n6.Out R, AdditionalOperands=@n9.Out)
   Multiply_0 = UE.Multiply:Audio by Float(AdditionalOperands=@ADSREnvelopeADSREnvelope_0.Out Envelope, PrimaryOperand=@Multiply_2.Out)
   Multiply_1 = UE.Multiply:Audio by Float(PrimaryOperand=@n6.Out L, AdditionalOperands=@n9.Out)
   Multiply_3 = UE.Multiply:Audio by Float(AdditionalOperands=@ADSREnvelopeADSREnvelope_0.Out Envelope, PrimaryOperand=@Multiply_1.Out)
  ]
  links [{dst="@UE.OutputFormat.Quad.Audio:1", src=Multiply_0.Out}, {dst="@UE.OutputFormat.Quad.Audio:0", src=Multiply_3.Out}]
 }
]
properties {
 assetImportData=&self:AssetImportData
 attenuationSettings=/AmbientAudioBase/Attenuations/Fort_Ambient_Athena_Att
 bLooping=true
 bOverrideConcurrency=true
 duration=10000
 numChannels=4
 outputFormat=Quad
 sampleRate=48000
 soundClassObject=/Game/Sounds/FortSoundClassesAndModes/Fort_Ambient_Global
 concurrencyOverrides {maxCount=2}
}
